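# Physics asset for the M_MED_Assemble_L male medium body cosmetic, covering its dynamic forearm, knee, side and shoulder-pad bones with capsule bodies.
# The left forearm and knee pieces constrain to the lower arm and calf, and both side pieces to tiny master-bone capsules, with narrow angular limits that favor one swing axis.
# The shoulder-pad master bones and left lower arm link to the upper spine with free angular motion.
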PhysicsAsset /BRCosmetics/Characters/Player/Male/Medium/Bodies/M_MED_Assemble_L/Meshes/M_MED_Assemble_L_Physics
bodies {
 calf_l [{shapeType=Capsule, length=5.0402, radius=4.547, center=<-0.3019 0.3008 -0.3715>, rotation=<rot -3.1425 -90 3.326>}]
 dyn_forearm_l [{shapeType=Capsule, length=10.1509, radius=8.2015, center=<3.4411 0.2303 5.2256>, rotation=<rot -4.7679 -90 17.6958>}]
 dyn_knee_l [{shapeType=Capsule, length=8.8513, radius=6.9524, center=<-0.0109 1.6128 2.5084>, rotation=<rot 0 91.8551 -179.3985>}]
 dyn_shoulderpad_l_MSTR [{shapeType=Capsule, length=0.505, radius=0.505, rotation=<rot 0 -90 0>}]
 dyn_shoulderpad_r_MSTR [{shapeType=Capsule, length=0.505, radius=0.505, rotation=<rot 0 -90 0>}]
 dyn_side_l [{shapeType=Capsule, length=11.3248, radius=7.403, center=<10.0126 -0.4773 0.8843>, rotation=<rot 0 85.1712 -82.0369>}]
 dyn_side_l_MSTR [{shapeType=Capsule, length=0.505, radius=0.505, rotation=<rot 0 -90 0>}]
 dyn_side_r [{shapeType=Capsule, length=11.3248, radius=7.403, center=<10.0126 -0.4773 0.8843>, rotation=<rot 0 85.1712 -82.0369>}]
 dyn_side_r_MSTR [{shapeType=Capsule, length=0.505, radius=0.505, rotation=<rot 0 -90 0>}]
 lowerarm_l [{shapeType=Capsule, length=11.3449, radius=6.9946, center=<10.2992 0.6438 -0.0927>, rotation=<rot 0 -90.1971 -88.6164>}]
 spine_05 [{shapeType=Capsule, length=12.1401, radius=10.1414, center=<5.2122 -1.9801 0.0253>, rotation=<rot 0 175.4792 -89.4278>}]
}
constraints [
 {bone1Name=dyn_forearm_l, bone2Name=lowerarm_l, swing1LimitDegrees=1, swing2LimitDegrees=8, twistLimitDegrees=1}
 {bone1Name=dyn_knee_l, bone2Name=calf_l, swing1LimitDegrees=1, swing2LimitDegrees=15, twistLimitDegrees=2}
 {
  bone1Name=lowerarm_l
  bone2Name=spine_05
  swing1LimitDegrees=45
  swing1Motion=Free
  swing2LimitDegrees=45
  swing2Motion=Free
  twistLimitDegrees=45
  twistMotion=Free
 }
 {
  bone1Name=dyn_shoulderpad_l_MSTR
  bone2Name=spine_05
  swing1LimitDegrees=45
  swing1Motion=Free
  swing2LimitDegrees=45
  swing2Motion=Free
  twistLimitDegrees=45
  twistMotion=Free
 }
 {
  bone1Name=dyn_shoulderpad_r_MSTR
  bone2Name=spine_05
  swing1LimitDegrees=45
  swing1Motion=Free
  swing2LimitDegrees=45
  swing2Motion=Free
  twistLimitDegrees=45
  twistMotion=Free
 }
 {bone1Name=dyn_side_l, bone2Name=dyn_side_l_MSTR, swing1LimitDegrees=1, swing2LimitDegrees=20, twistLimitDegrees=1}
 {bone1Name=dyn_side_r, bone2Name=dyn_side_r_MSTR, swing1LimitDegrees=1, swing2LimitDegrees=20, twistLimitDegrees=1}
]
