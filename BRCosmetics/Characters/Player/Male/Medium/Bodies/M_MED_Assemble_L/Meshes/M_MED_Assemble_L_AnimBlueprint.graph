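# Animation Blueprint for the Male Medium M_MED_Assemble_L cosmetic body, following the attached parent mesh through Copy Pose From Mesh.
# Weighted ten-target constraints position dyn_arm_l/r from spine, neck, clavicle and upper-arm twist bones, feeding a Control Rig.
# Additive bone-space offsets shift both shoulder pad master bones while bIsSkyDiving is true.
# A root-relative RigidBody simulation adds secondary motion last; leftover unconnected RigidBody and conversion nodes remain.
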AnimBlueprint /BRCosmetics/Characters/Player/Male/Medium/Bodies/M_MED_Assemble_L/Meshes/M_MED_Assemble_L_AnimBlueprint
target_skeleton=&/M_MED_Assemble_L_Skeleton
graphs [
 AnimGraph {nodes [
   n1 = RigidBody() {
    overridePhysicsAsset=&/M_MED_Assemble_L_Physics
    simulationSpace=BaseBoneSpace
    alphaBoolBlend {blendOption=Linear}
    baseBoneRef {boneName=root}
    componentLinearAccScale=<0.2 0.2 0.4>
    componentLinearVelScale=<0.1 0.1 0.2>
   }
   Component To Local(ComponentPose=@n1)
   Component To Local()
   Component To Local()
   Local To Component(LocalPose=Component To Local())
   RigidBody() {
    overridePhysicsAsset=&M_MED_Assemble_L_Physics
    simulationSpace=BaseBoneSpace
    alphaBoolBlend {blendOption=Linear}
    baseBoneRef {boneName=root}
    componentLinearAccScale=<0.2 0.2 0.4>
    componentLinearVelScale=<0.1 0.1 0.2>
   }
   n2 = "Constraint - dyn_arm_l (PPPPPPPPPP)"!(
    ComponentPose=Local To Component(LocalPose=Copy Pose From Mesh() {bUseAttachedParent=true})
    Alpha=1
    ConstraintWeights_0=0.4913
    ConstraintWeights_1:float=0
    ConstraintWeights_2:float=0.004
    ConstraintWeights_3:float=0.1176
    ConstraintWeights_4:float=0.2248
    ConstraintWeights_5:float=0.1233
    ConstraintWeights_6:float=0.0325
    ConstraintWeights_7:float=0
    ConstraintWeights_8:float=0.0053
    ConstraintWeights_9:float=0.0012
   ) -> (Pose) {
    alphaBoolBlend {blendOption=Linear}
    boneToModify {boneName=dyn_arm_l}
    constraintSetup [
     {targetBone {boneName=clavicle_l}}
     {targetBone {boneName=upperarm_twist_01_r}}
     {targetBone {boneName=spine_02}}
     {targetBone {boneName=spine_03}}
     {targetBone {boneName=spine_04}}
     {targetBone {boneName=spine_05}}
     {targetBone {boneName=upperarm_twist_01_l}}
     {targetBone {boneName=clavicle_r}}
     {targetBone {boneName=neck_02}}
     {targetBone {boneName=neck_01}}
    ]
    constraintWeights [1,1,1,1,1,1,1,1,1,1]
   }
   n3 = "Constraint - dyn_arm_r (PPPPPPPPPP)"!(
    ComponentPose=@n2
    Alpha=1
    ConstraintWeights_0=0
    ConstraintWeights_1:float=0.0325
    ConstraintWeights_2:float=0.004
    ConstraintWeights_3:float=0.1176
    ConstraintWeights_4:float=0.2248
    ConstraintWeights_5:float=0.1233
    ConstraintWeights_6:float=0
    ConstraintWeights_7:float=0.4913
    ConstraintWeights_8:float=0.0053
    ConstraintWeights_9:float=0.0012
   ) -> (Pose) {
    alphaBoolBlend {blendOption=Linear}
    boneToModify {boneName=dyn_arm_r}
    constraintSetup [
     {targetBone {boneName=clavicle_l}}
     {targetBone {boneName=upperarm_twist_01_r}}
     {targetBone {boneName=spine_02}}
     {targetBone {boneName=spine_03}}
     {targetBone {boneName=spine_04}}
     {targetBone {boneName=spine_05}}
     {targetBone {boneName=upperarm_twist_01_l}}
     {targetBone {boneName=clavicle_r}}
     {targetBone {boneName=neck_02}}
     {targetBone {boneName=neck_01}}
    ]
    constraintWeights [1,1,1,1,1,1,1,1,1,1]
   }
   n4 = Control Rig(Source=Component To Local(ComponentPose=@n3)) {
    alphaBoolBlend {blendOption=Linear}
    controlRigAssetReference {blueprintRigClass=&/M_MED_Assemble_L_CTRL_RIG.M_MED_Assemble_L_CTRL_RIG_C}
   }
   n5 = Get bIsSkyDiving()
   n6 = "Transform (Modify) Bone - Bone: dyn_shoulderpad_l_MSTR"(
    ComponentPose=Local To Component(LocalPose=@n4)
    Alpha="To Float (Boolean)"(InBool=@n5)
    Translation="5,-1.7,-2"
    Rotation="0,0,0"
    Scale="1,1,1"
   ) {
    rotationMode=BMM_Additive
    rotationSpace=BCS_BoneSpace
    translationMode=BMM_Additive
    translationSpace=BCS_BoneSpace
    alphaBoolBlend {blendOption=Linear}
    boneToModify {boneName=dyn_shoulderpad_l_MSTR}
    translation=<5 -1.7 -2>
   }
   n7 = "Transform (Modify) Bone - Bone: dyn_shoulderpad_r_MSTR"(ComponentPose=@n6, Alpha="To Float (Boolean)"(InBool=@n5), Translation="-5,-1.7,2", Rotation="0,0,0", Scale="1,1,1") {
    rotationMode=BMM_Additive
    rotationSpace=BCS_BoneSpace
    translationMode=BMM_Additive
    translationSpace=BCS_BoneSpace
    alphaBoolBlend {blendOption=Linear}
    boneToModify {boneName=dyn_shoulderpad_r_MSTR}
    translation=<-5 -1.7 2>
   }
   n8 = RigidBody(ComponentPose=@n7) {
    overridePhysicsAsset=&M_MED_Assemble_L_Physics
    simulationSpace=BaseBoneSpace
    alphaBoolBlend {blendOption=Linear}
    baseBoneRef {boneName=root}
    componentLinearAccScale=<0.2 0.2 0.4>
    componentLinearVelScale=<0.1 0.1 0.2>
   }
   Output Pose(Result=Component To Local(ComponentPose=@n8)) {layerGroup=None}
  ]}
 EventGraph {links [], nodes []}
]
variables []
