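# Control Rig for the M_MED_Assemble_L player body part that drives its dyn_ head, mouth and eyebrow bones from the character's head and face pose.
# Head rotation relative to spine_05 becomes a damped dyn_head rotation through clamped per-axis remaps.
# Jaw rotation lowers dyn_mouth, and each brow's height relative to its eye raises the matching dyn_eyebrow.
# A few brow-rotation remap nodes are left unconnected and drive nothing.
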ControlRigBlueprint /BRCosmetics/Characters/Player/Male/Medium/Bodies/M_MED_Assemble_L/Meshes/Parts/M_MED_Assemble_L_CtrlRig
graphs [
 RigVMFunctionLibrary {links [], nodes []}
 RigVMModel {nodes [
   RigUnit_BeginExecution = &RigUnit_BeginExecution()
   n1 = &RigVMFunction_MathQuaternionToEuler(
    RotationOrder=ZYX
    Value=&RigUnit_GetRelativeTransformForItem(Child=(Type=Bone,Name="head"), Parent=(Type=Bone,Name="spine_05")).RelativeTransform.Rotation
   )
   &RigUnit_ModifyTransforms(
    ItemToModify=((Item=(Type=Bone,Name="dyn_head"),Transform=(Rotation=(X=0,Y=0,Z=0,W=1),Translation=(X=0,Y=0,Z=0),Scale3D=(X=1,Y=1,Z=1))))
    Weight=1
    ItemToModify.0.Transform.Rotation=&RigVMFunction_MathQuaternionFromEuler(
     Euler=(X=0,Y=0,Z=0)
     Euler.Z=&RigVMFunction_MathFloatRemap(SourceMaximum=25, SourceMinimum=-20, TargetMaximum=5, TargetMinimum=-1, bClamp=true, Value=@n1.Result.Z).Result
     Euler.Y=&RigVMFunction_MathFloatRemap(SourceMaximum=-30, SourceMinimum=1, TargetMaximum=5, TargetMinimum=-1, bClamp=true, Value=@n1.Result.Y).Result
     Euler.X=&RigVMFunction_MathFloatRemap(SourceMaximum=47, SourceMinimum=-47, TargetMaximum=25, TargetMinimum=-25, bClamp=true, Value=@n1.Result.X).Result
    ).Result
    ExecutePin=@RigUnit_BeginExecution.ExecutePin
   )
   n2 = /Script/RigVM.RigVMFunction_MathFloatAbs(Value=&RigVMFunction_MathQuaternionToEuler(
     RotationOrder=ZYX
     Value=&RigUnit_GetRelativeTransformForItem(Child=(Type=Bone,Name="C_jaw"), Parent=(Type=Bone,Name="faceAttach")).RelativeTransform.Rotation
    ).Result.Z) {is=struct RigVMUnitNode, template="Absolute::Execute(in Value,out Result)"}
   > &RigUnit_ModifyTransforms(
    ItemToModify=((Item=(Type=Bone,Name="dyn_mouth"),Transform=(Rotation=(X=0,Y=0,Z=0,W=1),Translation=(X=0,Y=0,Z=0),Scale3D=(X=1,Y=1,Z=1))))
    Weight=1
    ItemToModify.0.Transform.Translation.Z=/Script/RigVM.RigVMFunction_MathFloatNegate(Value=&RigVMFunction_MathFloatRemap(SourceMaximum=25, SourceMinimum=3, TargetMaximum=1.5, TargetMinimum=0, bClamp=true, Value=@n2.Result).Result) {is=struct RigVMUnitNode, template="Negate::Execute(in Value,out Result)"}.Result
   )
   n3 = &RigUnit_GetRelativeTransformForItem(Child=(Type=Bone,Name="R_brow_mid"), Parent=(Type=Bone,Name="R_eye"))
   > &RigUnit_ModifyTransforms(
    ItemToModify=((Item=(Type=Bone,Name="dyn_eyebrow_r"),Transform=(Rotation=(X=0,Y=0,Z=0,W=1),Translation=(X=0,Y=0,Z=0),Scale3D=(X=1,Y=1,Z=1))))
    Weight=1
    ItemToModify.0.Transform.Translation.Z=&RigVMFunction_MathFloatRemap(SourceMaximum=2.7, SourceMinimum=1.5, TargetMaximum=4, TargetMinimum=0, bClamp=true, Value=@n3.RelativeTransform.Translation.Z).Result
   )
   n4 = &RigUnit_GetRelativeTransformForItem(Child=(Type=Bone,Name="L_brow_mid"), Parent=(Type=Bone,Name="L_eye"))
   > &RigUnit_ModifyTransforms(
    ItemToModify=((Item=(Type=Bone,Name="dyn_eyebrow_l"),Transform=(Rotation=(X=0,Y=0,Z=0,W=1),Translation=(X=0,Y=0,Z=0),Scale3D=(X=1,Y=1,Z=1))))
    Weight=1
    ItemToModify.0.Transform.Translation.Z=&RigVMFunction_MathFloatRemap(SourceMaximum=2.7, SourceMinimum=1.5, TargetMaximum=4, TargetMinimum=0, bClamp=true, Value=@n4.RelativeTransform.Translation.Z).Result
   )
   &RigUnit_GetRelativeTransformForItem(Child=(Type=Bone,Name="C_brow_mid"), Parent=(Type=Bone,Name="L_eye"))
   &RigVMFunction_MathFloatAbs(Value=&RigVMFunction_MathQuaternionToEuler(RotationOrder=ZYX, Value=@n4.RelativeTransform.Rotation).Result.Y) {template="Absolute::Execute(in Value,out Result)"}
   &RigVMFunction_MathFloatRemap(
    SourceMaximum=25
    SourceMinimum=3
    TargetMaximum=3
    TargetMinimum=0
    bClamp=true
    Value=&RigVMFunction_MathFloatAbs(Value=&RigVMFunction_MathQuaternionToEuler(RotationOrder=ZYX, Value=@n3.RelativeTransform.Rotation).Result.Z) {template="Absolute::Execute(in Value,out Result)"}.Result
   )
   n5 = &RigVMFunction_MathQuaternionToEuler(RotationOrder=ZYX, Value=(X=0,Y=0,Z=0,W=1))
   &RigVMFunction_MathQuaternionFromEuler(
    Euler=(X=0,Y=0,Z=0)
    Euler.Z=&RigVMFunction_MathFloatRemap(
     SourceMaximum=25
     SourceMinimum=3
     TargetMaximum=3
     TargetMinimum=0
     bClamp=true
     Value=&RigVMFunction_MathFloatAbs(Value=@n5.Result.Z) {template="Absolute::Execute(in Value,out Result)"}.Result
    ).Result
   )
  ]}
]
hierarchy [
 root {loc=<0 0 0>, rot=<0 0 0 1>}
 pelvis {parent=root, loc=<0 -0.7071 93.1191>, rot=<0.0127 0.707 -0.0127 -0.707>}
 spine_01 {parent=pelvis, loc=<6.7494 -0.6388 0>, rot=<0 0 0.034 -0.9994>}
 spine_02 {parent=spine_01, loc=<7.2518 -0.035 0>, rot=<0 0 0.0238 -0.9997>}
 spine_03 {parent=spine_02, loc=<6.0104 0 0>, rot=<0 0 0.0149 -0.9999>}
 spine_04 {parent=spine_03, loc=<8.1024 0 0>, rot=<0 0 -0.0132 -0.9999>}
 spine_05 {parent=spine_04, loc=<10.0146 0 0>, rot=<0 0 -0.0015 -1>}
 clavicle_l {parent=spine_05, loc=<4.3575 -0.1466 -2.3223>, rot=<0.0465 -0.7289 -0.009 -0.6829>}
 upperarm_l {parent=clavicle_l, loc=<15.5111 0 0>, rot=<0.0305 -0.536 -0.0057 -0.8437>}
 upperarm_twist_01_l {parent=upperarm_l, loc=<0.0041 0 0>, rot=<0 0 0 -1>}
 clavicle_r {parent=spine_05, loc=<4.3575 -0.1466 2.3223>, rot=<0.7289 0.0465 -0.6829 0.009>}
 upperarm_r {parent=clavicle_r, loc=<-15.5111 0 0>, rot=<0.0305 -0.536 -0.0057 -0.8437>}
 upperarm_twist_01_r {parent=upperarm_r, loc=<-0.0041 0 0>, rot=<0 0 0 1>}
 neck_01 {parent=spine_05, loc=<11.3138 0.1529 0>, rot=<0 0 -0.2003 -0.9797>}
 neck_02 {parent=neck_01, loc=<4.5627 0 0>, rot=<0 0 0.0504 -0.9987>}
 head {parent=neck_02, loc=<5.3914 0 0>, rot=<0 0 0.115 -0.9934>}
 faceAttach {parent=head, loc=<-1.7906 -1.603 0.0036>, rot=<0.0082 -0.7071 -0.0082 -0.7071>, scale=<0.89 0.89 0.89>}
 C_brow_mid {parent=faceAttach, loc=<-0.0193 11.1161 6.5383>, rot=<0 0.0183 -0.0036 -0.9998>}
 L_brow_mid {parent=faceAttach, loc=<3.2585 10.6172 8.0786>, rot=<-0.0002 -0.0047 0.2553 -0.9669>}
 L_brow_outer {parent=faceAttach, loc=<5.73 8.8025 8.1005>, rot=<0.1894 0.2347 0.3767 -0.8758>}
 L_cheek_inner {parent=faceAttach, loc=<4.3171 9.4336 2.0731>, rot=<-0.0228 -0.1484 0.3324 -0.9311>}
 C_lip_upper_mid {parent=faceAttach, loc=<0.0248 11.1369 -0.5028>, rot=<0.2225 0.0009 -0.0003 -0.9749>}
 L_lip_upper_outer {parent=faceAttach, loc=<1.7089 10.4032 -0.242>, rot=<0.0394 -0.0629 0.259 -0.963>}
 L_lip_corner {parent=faceAttach, loc=<2.7097 9.6174 -0.4462>, rot=<0.0026 0.0084 -0.0146 -0.9999>}
 C_nose_bridge {parent=faceAttach, loc=<-0.0358 12.2872 3.2888>, rot=<-0.2218 -0.0062 -0.0017 -0.9751>}
 R_brow_mid {parent=faceAttach, loc=<-3.2498 10.6185 8.0841>, rot=<-0.0001 0.0043 -0.2553 -0.9669>}
 R_brow_outer {parent=faceAttach, loc=<-5.7233 8.8052 8.0744>, rot=<0.118 -0.1447 -0.4069 -0.8942>}
 R_cheek_inner {parent=faceAttach, loc=<-4.2977 9.4402 2.0626>, rot=<-0.0228 0.1484 -0.3324 -0.9311>}
 R_lip_corner {parent=faceAttach, loc=<-2.5549 9.7202 -0.4934>, rot=<0.0029 -0.008 0.0146 -0.9999>}
 C_jaw {parent=faceAttach, loc=<-0.0119 3.2605 0.3823>, rot=<0 0 0 1>}
 L_lip_lower_outer {parent=C_jaw, loc=<1.5062 7.0528 -0.8964>, rot=<0.0166 0.0643 0.3205 -0.9449>}
 C_lip_lower_mid {parent=C_jaw, loc=<0.018 7.6348 -1.0482>, rot=<0 -0.0179 -0.0038 -0.9998>}
 R_lip_lower_outer {parent=C_jaw, loc=<-1.5646 7.1146 -1.0273>, rot=<0.0075 -0.0145 -0.2828 -0.9591>}
 teeth_lower {parent=C_jaw, loc=<-0.026 5.1259 -2.5487>, rot=<0 0 0 1>}
 tongue {parent=C_jaw, loc=<0.0084 1.4856 -1.0316>, rot=<-0.1568 0 0 -0.9876>}
 R_eye {parent=faceAttach, loc=<-3.1065 8.8293 6.2365>, rot=<0 0 0.9998 -0.0187>}
 R_eye_lid_upper_mid {parent=R_eye, loc=<0.1994 -1.6412 0.3877>, rot=<0.0523 -0.0009 0.0167 -0.9985>}
 R_eye_lid_lower_mid {parent=R_eye, loc=<-0.1442 -1.5052 -0.8207>, rot=<0.0329 -0.0005 0.0167 -0.9993>}
 L_eye {parent=faceAttach, loc=<3.1146 8.8293 6.2365>, rot=<0 0 0.9998 0.0187>}
 L_eye_lid_upper_mid {parent=L_eye, loc=<-0.2226 -1.6322 0.3638>, rot=<0.0523 0 0 -0.9986>}
 L_eye_lid_lower_mid {parent=L_eye, loc=<0.1562 -1.4858 -0.8645>, rot=<0.0329 0 0 -0.9995>}
 teeth_upper {parent=faceAttach, loc=<0.0061 9.4557 0.1849>, rot=<0 0 0 1>}
 R_lip_upper_outer {parent=faceAttach, loc=<-1.7315 10.4594 -0.2337>, rot=<0.0383 0.0661 -0.2602 -0.9625>}
 dyn_head {parent=neck_01, loc=<4.5627 0 0>, rot=<0 0 0.0504 -0.9987>}
 dyn_vtech {parent=dyn_head, loc=<4.5116 -10.1772 -0.0735>, rot=<0.3389 -0.6206 0.6206 -0.3389>}
 dyn_mouth {parent=dyn_head, loc=<3.4689 6.5121 -0.1455>, rot=<-0.0775 -0.7025 0.0774 -0.7032>}
 dyn_eyebrow_l {parent=dyn_head, loc=<16.2634 7.1347 -5.4495>, rot=<-0.0784 -0.7027 0.0784 -0.7027>}
 dyn_eyebrow_r {parent=dyn_head, loc=<16.6668 7.2194 5.7778>, rot=<-0.0784 -0.7027 0.0784 -0.7027>}
 dyn_ribbon_low {parent=dyn_head, loc=<15.5977 -15.6473 -1.2234>, rot=<0.4415 -0.5523 0.5523 -0.4415>}
 dyn_ribbon_high {parent=dyn_head, loc=<23.7248 -10.8295 -3.673>, rot=<0.6193 -0.3413 0.3413 -0.6193>}
 dyn_attenna_base {parent=dyn_head, loc=<19.8666 -15.4463 6.4406>, rot=<0.5756 -0.4028 0.5098 -0.4965>}
 dyn_attenna_twist {parent=dyn_attenna_base, loc=<7.2312 0.0515 -0.2707>, rot=<0 0.0048 -0.0004 -1>}
]
