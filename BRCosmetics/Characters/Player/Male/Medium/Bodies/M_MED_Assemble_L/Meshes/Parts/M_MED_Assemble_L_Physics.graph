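# Physics asset for the M_MED_Assemble_L male medium player cosmetic, covering its dynamic head, antenna base and two ribbon bones, likely for secondary motion.
# Each of the four bones has a single capsule body.
# Constraints form a star around dyn_head, tying the antenna base and both ribbons directly to it with narrow swing limits and near-zero twist limits.
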PhysicsAsset /BRCosmetics/Characters/Player/Male/Medium/Bodies/M_MED_Assemble_L/Meshes/Parts/M_MED_Assemble_L_Physics
bodies {
 dyn_attenna_base [{shapeType=Capsule, length=8.967, radius=1.5789, center=<5.1595 0.1027 -0.3778>, rotation=<rot 0 -88.8558 -85.8128>}]
 dyn_head [{shapeType=Capsule, length=2.985, radius=3.3518, center=<12.712 -3.274 -0.2633>, rotation=<rot 0 100.5251 -85.2247>}]
 dyn_ribbon_high [{shapeType=Capsule, length=4.3833, radius=1.9512, center=<0.8965 0 0.8158>, rotation=<rot 42.3027 0 0>}]
 dyn_ribbon_low [{shapeType=Capsule, length=4.5831, radius=2.1509, center=<1.1956 0.0687 0.4668>}]
}
constraints [
 {bone1Name=dyn_ribbon_high, bone2Name=dyn_head, swing1LimitDegrees=10, swing2LimitDegrees=20, twistLimitDegrees=1}
 {bone1Name=dyn_ribbon_low, bone2Name=dyn_head, swing1LimitDegrees=10, swing2LimitDegrees=20, twistLimitDegrees=1}
 {bone1Name=dyn_attenna_base, bone2Name=dyn_head, swing1LimitDegrees=15, swing2LimitDegrees=20, twistLimitDegrees=1}
]
